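; Construction script Blueprint for the Sungold large pottery prop, adding secondary color randomization on top of its parent's construction logic.
; It instantiates the secondary swatch palette class, then picks a random swatch index, or keeps the set index when randomization is disabled.
; The chosen swatch goes into building texture data slot 1, whose tint is passed to SetColoronSecondaryMesh with the static mesh component.
; The event graph events are empty.
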(event EventBeginPlay)

(event Collision|EventActorBeginOverlap
  (params (OtherActor :type Actor%20Object%20Reference)))

(event EventTick
  (params (DeltaSeconds :type Float%20%28single-precision%29)))

(fn ConstructionScript
  (bind _secondarypalette (Variables|Default|GetSecondaryPalette))
  (bind _swatch_palette (Variables|/Game/Environments/Apollo/Blueprints/RandomColor_Prop/RandomColor_Prop_Palette.RandomColor_Prop_Palette_C|GetSwatch_Palette _secondarypalette))
  (|CallParentFunction)
  (bind _returnvalue (Game|ConstructObjectfromClass (Variables|Default|GetSecondarySwatchPalette) self))
  (Variables|Default|SetSecondaryPalette _returnvalue)
  (if (Variables|Default|GetDisableSecondaryColorRandomization)
    (Variables|Default|SetSecondaryColorIndex (Variables|Default|GetSecondaryColorIndex))
    (else
      (Variables|Default|SetSecondaryColorIndex (Math|Random|RandomIntegerinRange 0 (- (Utilities|Array|Length _swatch_palette) 1)))))
  (CallFunction|SetBuildingTextureData 1 (Utilities|Array|Get(acopy) _swatch_palette (Variables|Default|GetSecondaryColorIndex)))
  (bind _outtexturedataarray (Building|GetBuildingTextureData))
  (if (Utilities|Array|IsValidIndex _outtexturedataarray 1)
    (Default|SetColoronSecondaryMesh (Math|Conversions|ToLinearColor(Color) (Variables|/Script/FortniteGame.BuildingTextureData|GetTintColor (Utilities|Array|Get(acopy) _outtexturedataarray 1))) (Variables|Debug|GetStaticMeshComponent))))
